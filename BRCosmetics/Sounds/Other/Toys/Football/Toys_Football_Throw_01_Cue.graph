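# Sound cue for the throw sound of the football toy cosmetic.
# A single wave player feeds the output directly, with no randomization, and plays the Toy_Football_Throw_02 wave despite the cue's 01 name.
# It uses the Athena emote attenuation and the emote music sound class, with a slight volume boost.
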SoundCue /BRCosmetics/Sounds/Other/Toys/Football/Toys_Football_Throw_01_Cue
root=WavePlayer_0
nodes [{id=WavePlayer_0, properties {soundWaveAssetPtr=&/Toy_Football_Throw_02}}]
properties {
 attenuationSettings=&Fort_Emotes_Athena_Att
 maxDistance=1600
 soundClassObject=&Fort_Music_Emote
 volumeMultiplier=1.25
}
